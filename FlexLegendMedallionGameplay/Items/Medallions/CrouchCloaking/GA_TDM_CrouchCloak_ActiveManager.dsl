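; Gameplay ability for the crouch-cloak medallion's active state that applies a custom camera mode on the non-authority side.
; When a scalable-float toggle allows it, activation pushes the camera mode, after an optional holster-delay timer.
; It follows weapon-equipped and targeting-changed delegates, popping and re-pushing between default and targeting modes as targeting changes.
; Ending pops the camera mode, clears the timer and unbinds the delegates.
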(event Ability|EventActivateAbility
  (bind _sf_holsterdelay (Variables|Default|GetSF_HolsterDelay))
  (bind _returnvalue (ScalableFloat|GetValueatLevel _sf_holsterdelay))
  ;[0.1039 0.524 0.07259 0.5] Activate Ability
  (|CallParentFunction)
  (Ability|HasAuthority
    (:False
      (if (ScalableFloat|GetValueasBool (Variables|Default|GetSF_UseCustomCameraMode))
        (if (> _returnvalue 0)
          (bind _returnvalue_2 (Utilities|Time|SetTimerbyEvent (EventDispatchers|CreateEvent :!event Custom|InitializeCustomCameraMode) _returnvalue))
          (Variables|Default|SetWaitForHolsterTimerHandle _returnvalue_2)
          (else
            (CallFunction|InitializeCustomCameraMode)))))))

(event Custom|InitializeCustomCameraMode
  ;[0 0.524 0.1282 0.5] Initialization: Bind to events and push initial camera mode
  (bind _asfort_player_pawn (Utilities|Casting|CastToFortPlayerPawn (Ability|GetActivatingPawn)))
  (bind _output_get (Variables|Default|SetFortPlayerPawn _asfort_player_pawn))
  (FortPawn|BindEventtoOnWeaponEquippedDelegate _output_get (EventDispatchers|CreateEvent :!event Custom|OnWeaponChanged))
  (bind _currentweapon (Variables|/Script/FortniteGame.FortPawn|GetCurrentWeapon _asfort_player_pawn)
    (:then
      (CallFunction|PushCrouchCloakCameraMode (FortWeapon|IsTargeting _currentweapon)))
    (:else
      (CallFunction|PushCrouchCloakCameraMode false)))
  (CallFunction|OnWeaponChanged _currentweapon))

(event Custom|OnWeaponChanged
  (params (NewWeapon :type Fort%20Weapon%20Object%20Reference) (PrevWeapon :type Fort%20Weapon%20Object%20Reference))
  (bind _outputdelegate (EventDispatchers|CreateEvent :!event Custom|OnTargetingSet))
  ;[0.1662 0.7031 0.4447 0.5] On Weapon Changed: Update binding to targeting change, and update IsTargeting
  (Utilities|FlowControl|Sequence
    (:then_0
      (Utilities|IsValid PrevWeapon
        (:Is%20Valid
          (EventDispatchers|UnbindEventfromOnTargetingChanged PrevWeapon _outputdelegate))))
    (:then_1
      (Utilities|IsValid NewWeapon
        (:Is%20Valid
          (EventDispatchers|BindEventtoOnTargetingChanged NewWeapon _outputdelegate)
          (CallFunction|OnTargetingSet (FortWeapon|IsTargeting NewWeapon)))))))

(event Custom|OnTargetingSet
  (params (bNewIsTargeting :type Boolean))
  ;[0.2153 0.7031 0.6961 0.5] On Targeting Set: If targeting state is different, remove current camera mode and push new one
  (if (!= bNewIsTargeting (Variables|Default|GetIs%20Targeting))
    (Camera|PopCameraMode)
    (bind _output_get (Variables|Default|SetIs%20Targeting bNewIsTargeting))
    (CallFunction|PushCrouchCloakCameraMode _output_get)))

(event Custom|PushCrouchCloakCameraMode
  (params (bIsTargeting :type Boolean))
  ;[0.6183 0.2399 0.7031 0.5] Push camera mode based on targeting state
  (if bIsTargeting
    (Camera|PushCameraMode (Variables|Default|GetTargetingCameraMode))
    (else
      (Camera|PushCameraMode (Variables|Default|GetDefaultCameraMode)))))

(event Ability|EventOnEndAbility
  (params (bWasCancelled :type Boolean))
  ;[0.7781 0.3547 0.05928 0.5] On End Ability: Cleanup binds and remove camera mode
  (|CallParentFunction bWasCancelled)
  (Ability|HasAuthority
    (:False
      (Camera|PopCameraMode)
      (Utilities|Time|ClearandInvalidateTimerbyHandle (Variables|Default|GetWaitForHolsterTimerHandle))
      (bind _fortplayerpawn (Variables|Default|GetFortPlayerPawn)
        (:then
          (FortPawn|UnbindEventfromOnWeaponEquippedDelegate _fortplayerpawn (EventDispatchers|CreateEvent :!event Custom|OnWeaponChanged)))
        (:else
          _))
      (bind _currentweapon (Variables|Default|GetCurrentWeapon)
        (:then
          (EventDispatchers|UnbindEventfromOnTargetingChanged _currentweapon (EventDispatchers|CreateEvent :!event Custom|OnTargetingSet)))
        (:else)))))
